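# Post-copy animation blueprint for the TeriyakiTech medium male body cosmetic, driving its extra knee pad and bag bones from the attached parent mesh's pose.
# The copied pose runs through a DeformRig and weighted bone constraints into cached pose 'pre'.
# A bag-correction branch adds offsets and Look Ats toward hip-and-thigh-constrained targets, but its bool blend defaults to plain 'pre'.
# A final CtrlRig exposes a crouch input; the event graph is empty.
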AnimBlueprint /BRCosmetics/Characters/Player/Male/Medium/Bodies/M_MED_TeriyakiTech/Meshes/M_MED_TeriyakiTech_AnimBP
target_skeleton=&/M_MED_TeriyakiTech_Skeleton
graphs [
 AnimGraph {nodes [
   n1 = "Transform (Modify) Bone - Bone: extra_bag_fr"!(
    ComponentPose=Local To Component(LocalPose=Use cached pose 'pre'())
    bAlphaBoolEnabled:bool=True
    Translation="-0.647,0.924,1.004"
    Rotation="14.595,-5.578,-20.645"
    Scale="1,1,1"
   ) -> (Pose) {
    alphaInputType=Bool
    rotationMode=BMM_Additive
    rotationSpace=BCS_BoneSpace
    translationMode=BMM_Additive
    translationSpace=BCS_BoneSpace
    alphaBoolBlend {blendInTime=0.1, blendOption=CircularInOut, blendOutTime=0.1}
    boneToModify {boneName=extra_bag_fr}
    rotation=<rot 14.595 -5.578 -20.645>
    translation=<-0.647 0.924 1.004>
   }
   n2 = "Constraint - extra_bag_fr_target (PPPPPPPP)"!(
    ComponentPose=@n1
    Alpha=1
    ConstraintWeights_0=0.1806
    ConstraintWeights_1:float=0.0366
    ConstraintWeights_2:float=0.2894
    ConstraintWeights_3:float=0
    ConstraintWeights_4:float=0.0837
    ConstraintWeights_5:float=0.0366
    ConstraintWeights_6:float=0.2894
    ConstraintWeights_7:float=0.0837
   ) -> (Pose) {
    alphaBoolBlend {blendOption=Linear}
    boneToModify {boneName=extra_bag_fr_target}
    constraintSetup [
     {targetBone {boneName=pelvis}}
     {targetBone {boneName=thigh_l}}
     {targetBone {boneName=thigh_twist_01_l}}
     {targetBone {boneName=deform_glute_l}}
     {targetBone {boneName=deform_groin_l}}
     {targetBone {boneName=thigh_r}}
     {targetBone {boneName=thigh_twist_01_r}}
     {targetBone {boneName=deform_groin_r}}
    ]
    constraintWeights [1,1,1,1,1,1,1,1]
   }
   n3 = Look At - Bone: extra_bag_fr(ComponentPose=@n2, Alpha=1) {
    alphaBoolBlend {blendOption=Linear}
    boneToModify {boneName=extra_bag_fr}
    lookAtLocation=<2.0585 9.021 -8.8406>
    lookAtTarget {boneReference {boneName=extra_bag_fr_target}}
   }
   n4 = "Transform (Modify) Bone - Bone: extra_bag_fr_side"!(ComponentPose=@n3, bAlphaBoolEnabled:bool=True, Translation="3.447,-0.190,-0.088", Rotation="-37.237,-0.514,4.537", Scale="1,1,1") -> (Pose) {
    alphaInputType=Bool
    rotationMode=BMM_Additive
    rotationSpace=BCS_BoneSpace
    translationMode=BMM_Additive
    translationSpace=BCS_BoneSpace
    alphaBoolBlend {blendInTime=0.1, blendOption=CircularInOut, blendOutTime=0.1}
    boneToModify {boneName=extra_bag_fr_side}
    rotation=<rot -37.237 -0.514 4.537>
    translation=<3.447 -0.19 -0.088>
   }
   n5 = "Constraint - extra_bag_fr_side_target (PPPPPPPP)"!(
    ComponentPose=@n4
    Alpha=1
    ConstraintWeights_0=0.5633
    ConstraintWeights_1:float=0.0016
    ConstraintWeights_2:float=0.0046
    ConstraintWeights_3:float=0.0121
    ConstraintWeights_4:float=0.3561
    ConstraintWeights_5:float=0.0215
    ConstraintWeights_6:float=0.0369
    ConstraintWeights_7:float=0.004
   ) -> (Pose) {
    alphaBoolBlend {blendOption=Linear}
    boneToModify {boneName=extra_bag_fr_side_target}
    constraintSetup [
     {targetBone {boneName=pelvis}}
     {targetBone {boneName=spine_01}}
     {targetBone {boneName=spine_02}}
     {targetBone {boneName=thigh_l}}
     {targetBone {boneName=thigh_twist_01_l}}
     {targetBone {boneName=deform_glute_l}}
     {targetBone {boneName=deform_groin_l}}
     {targetBone {boneName=thigh_twist_01_r}}
    ]
    constraintWeights [1,1,1,1,1,1,1,1]
   }
   n6 = Look At - Bone: extra_bag_fr_side(ComponentPose=@n5, Alpha=1) {
    alphaBoolBlend {blendOption=Linear}
    boneToModify {boneName=extra_bag_fr_side}
    lookAtLocation=<0.6579 3.2296 -2.8733>
    lookAtTarget {boneReference {boneName=extra_bag_fr_side_target}}
   }
   n7 = Control Rig!(
    Source=Blend Poses by bool(BlendPose_0=Component To Local(ComponentPose=@n6), BlendPose_1=Use cached pose 'pre'()) {blendType=Linear}
    Alpha
    crouch:bool=false
   ) -> (Pose) {
    alphaBoolBlend {blendOption=Linear}
    controlRigAssetReference {blueprintRigClass=&/M_MED_TeriyakiTech_CtrlRig.M_MED_TeriyakiTech_CtrlRig_C}
   }
   Output Pose(Result=@n7) {layerGroup=None}
   n8 = Control Rig!(Source=Copy Pose From Mesh() {bUseAttachedParent=true}, Alpha) -> (Pose) {
    alphaBoolBlend {blendOption=Linear}
    controlRigAssetReference {blueprintRigClass=&/M_MED_TeriyakiTech_DeformRig.M_MED_TeriyakiTech_DeformRig_C}
   }
   n9 = "Constraint - extra_knee_pad_l (PP)"!(ComponentPose=Local To Component(LocalPose=@n8), Alpha=1, ConstraintWeights_0=0.5, ConstraintWeights_1:float=0.5) -> (Pose) {
    alphaBoolBlend {blendOption=Linear}
    boneToModify {boneName=extra_knee_pad_l}
    constraintSetup [{targetBone {boneName=thigh_l}}, {targetBone {boneName=calf_l}}]
    constraintWeights [1,1]
   }
   n10 = "Constraint - extra_knee_pad_r (PP)"!(ComponentPose=@n9, Alpha=1, ConstraintWeights_0=0.5, ConstraintWeights_1:float=0.5) -> (Pose) {
    alphaBoolBlend {blendOption=Linear}
    boneToModify {boneName=extra_knee_pad_r}
    constraintSetup [{targetBone {boneName=thigh_r}}, {targetBone {boneName=calf_r}}]
    constraintWeights [1,1]
   }
   n11 = "Constraint - extra_bag_bk (PPPP)"!(
    ComponentPose=@n10
    Alpha=1
    ConstraintWeights_0=0.8
    ConstraintWeights_1:float=0.0515
    ConstraintWeights_2:float=0.098
    ConstraintWeights_3:float=0.0505
   ) -> (Pose) {
    alphaBoolBlend {blendOption=Linear}
    boneToModify {boneName=extra_bag_bk}
    constraintSetup [{targetBone {boneName=pelvis}}, {targetBone {boneName=spine_01}}, {targetBone {boneName=spine_02}}, {targetBone {boneName=spine_03}}]
    constraintWeights [1,1,1,1]
   }
   n12 = "Constraint - extra_bag_fr (PPPP)"!(
    ComponentPose=@n11
    Alpha=1
    ConstraintWeights_0=0.8
    ConstraintWeights_1:float=0.0515
    ConstraintWeights_2:float=0.098
    ConstraintWeights_3:float=0.0505
   ) -> (Pose) {
    alphaBoolBlend {blendOption=Linear}
    boneToModify {boneName=extra_bag_fr}
    constraintSetup [{targetBone {boneName=pelvis}}, {targetBone {boneName=spine_01}}, {targetBone {boneName=spine_02}}, {targetBone {boneName=spine_03}}]
    constraintWeights [1,1,1,1]
   }
   n13 = "Constraint - extra_bag_fr_side (PPPP)"!(
    ComponentPose=@n12
    Alpha=1
    ConstraintWeights_0=0.8
    ConstraintWeights_1:float=0.0515
    ConstraintWeights_2:float=0.098
    ConstraintWeights_3:float=0.0505
   ) -> (Pose) {
    alphaBoolBlend {blendOption=Linear}
    boneToModify {boneName=extra_bag_fr_side}
    constraintSetup [{targetBone {boneName=pelvis}}, {targetBone {boneName=spine_01}}, {targetBone {boneName=spine_02}}, {targetBone {boneName=spine_03}}]
    constraintWeights [1,1,1,1]
   }
   Save cached pose 'pre'(Pose=Component To Local(ComponentPose=@n13)) {cacheName=pre}
  ]}
 EventGraph {nodes [Event BlueprintUpdateAnimation()]}
]
variables []
