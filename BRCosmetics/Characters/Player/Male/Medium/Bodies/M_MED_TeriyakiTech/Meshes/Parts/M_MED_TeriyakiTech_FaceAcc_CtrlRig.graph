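# Control Rig for the TeriyakiTech male medium body's face accessory that rotates the extra_head_ring bone on the head.
# The RotRing variable is wrapped modulo 360, multiplied by 100 and converted from an X-axis Euler angle to a quaternion for a single Set Rotation.
# The rig does not advance RotRing itself, so any spin depends on the value supplied to it.
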ControlRigBlueprint /BRCosmetics/Characters/Player/Male/Medium/Bodies/M_MED_TeriyakiTech/Meshes/Parts/M_MED_TeriyakiTech_FaceAcc_CtrlRig
graphs [
 RigVMFunctionLibrary {links [], nodes []}
 RigVMModel {nodes [
   RigUnit_BeginExecution = &RigUnit_BeginExecution()
   n1 = /Script/RigVM.RigVMFunction_MathDoubleMod(B=360, A=RigVMVariableNode() {variable=RotRing}.Value) {is=struct RigVMUnitNode, template="Modulo::Execute(in A,in B,out Result)"}
   RigUnit_SetRotation = &RigUnit_SetRotation(
    Item=(Type=Bone,Name="extra_head_ring")
    Weight=1
    bInitial=False
    Value=&RigVMFunction_MathQuaternionFromEuler(Euler=(X=0,Y=0,Z=0), Euler.X=&RigVMFunction_MathDoubleMul(B=100, A=@n1.Result).Result).Result
    ExecutePin=@RigUnit_BeginExecution.ExecutePin
   )
  ]}
]
hierarchy [
 root {loc=<0 0 0>, rot=<0 0 0 1>}
 pelvis {parent=root, loc=<0 -0.7071 93.1191>, rot=<0.0127 0.707 -0.0127 -0.707>}
 spine_01 {parent=pelvis, loc=<6.7494 -0.6388 0>, rot=<0 0 0.034 -0.9994>}
 spine_02 {parent=spine_01, loc=<7.2518 -0.035 0>, rot=<0 0 0.0238 -0.9997>}
 spine_03 {parent=spine_02, loc=<6.0104 0 0>, rot=<0 0 0.0149 -0.9999>}
 spine_04 {parent=spine_03, loc=<8.1024 0 0>, rot=<0 0 -0.0132 -0.9999>}
 spine_05 {parent=spine_04, loc=<10.0146 0 0>, rot=<0 0 -0.0015 -1>}
 neck_01 {parent=spine_05, loc=<11.3138 0.1529 0>, rot=<0 0 -0.2003 -0.9797>}
 neck_02 {parent=neck_01, loc=<4.5627 0 0>, rot=<0 0 0.0504 -0.9987>}
 head {parent=neck_02, loc=<5.3914 0 0>, rot=<0 0 0.115 -0.9934>}
 dyn_earFlap_l {parent=head, loc=<5.1188 2.3549 -13.1866>, rot=<-0.1409 0.6812 0.7037 0.1445>}
 dyn_earFlap_01_l {parent=dyn_earFlap_l, loc=<5.095 0.259 0.1749>, rot=<0 0 0 1>}
 dyn_earFlap_r {parent=head, loc=<4.665 2.3901 13.2711>, rot=<0.1234 0.6789 0.7118 -0.1309>}
 dyn_earFlap_01_r {parent=dyn_earFlap_r, loc=<6.1518 0.0999 0.2992>, rot=<0 0 0 1>}
 extra_head_ring {parent=head, loc=<20.2938 2.9253 -0.016>, rot=<0.9997 -0.0253 0 0>}
 clavicle_l {parent=spine_05, loc=<4.3575 -0.1466 -2.3223>, rot=<0.0465 -0.7289 -0.009 -0.6829>}
 upperarm_l {parent=clavicle_l, loc=<15.5111 0 0>, rot=<0.0226 -0.3234 -0.0088 -0.9459>}
 clavicle_r {parent=spine_05, loc=<4.3575 -0.1466 2.3223>, rot=<0.7289 0.0465 -0.6829 0.009>}
 upperarm_r {parent=clavicle_r, loc=<-15.5111 0 0>, rot=<0.0226 -0.3234 -0.0088 -0.9459>}
]
variables [RotRing {type=double, default=0}]
