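# Physics asset for the face accessory of the male TeriyakiTech cosmetic character, covering its left and right ear flaps and the head.
# Capsules form each dynamic ear flap body, while two large capsules form the head body.
# Each flap is constrained to the head with very tight swing and twist limits, allowing only slight movement.
# The sides are asymmetric in capsule layout and swing range.
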PhysicsAsset /BRCosmetics/Characters/Player/Male/Medium/Bodies/M_MED_TeriyakiTech/Meshes/Parts/M_MED_TeriyakiTech_FaceAcc_Physics
bodies {
 dyn_earFlap_l [
  {shapeType=Capsule, length=6, radius=2, shapeName=dyn_earFlap_l_capsule, center=<3.909 0.135 0>}
  {shapeType=Capsule, length=6, radius=3, shapeName=dyn_earFlap_01_l_capsule, center=<8.909 -0.0869 0>}
  {shapeType=Capsule, length=10, radius=3, shapeName=dyn_earFlap_l_capsule, center=<9.01 -1.9223 0>, rotation=<rot -90 179.9604 -179.9605>}
 ]
 dyn_earFlap_r [
  {shapeType=Capsule, length=5, radius=2, shapeName=dyn_earFlap_r_capsule, center=<4.4522 -0.1282 0.3468>}
  {
   shapeType=Capsule
   length=13.5
   radius=4
   shapeName=dyn_earFlap_01_r_capsule
   center=<12.7166 -3.4518 -1.547>
   rotation=<rot 90 180 -179.9999>
  }
 ]
 head [
  {shapeType=Capsule, length=24.6488, radius=11.5, center=<0 2.3335 0>, rotation=<rot -90 180 180>}
  {shapeType=Capsule, length=20, radius=15, shapeName=head_capsule, center=<0 -16.5806 0>}
 ]
}
constraints [
 {bone1Name=dyn_earFlap_l, bone2Name=head, swing1LimitDegrees=10, swing2LimitDegrees=1, twistLimitDegrees=1}
 {bone1Name=dyn_earFlap_r, bone2Name=head, swing1LimitDegrees=5, swing2LimitDegrees=1, twistLimitDegrees=1}
]
