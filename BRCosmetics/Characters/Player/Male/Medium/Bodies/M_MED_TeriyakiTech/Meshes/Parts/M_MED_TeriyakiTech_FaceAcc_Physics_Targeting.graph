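# Physics asset for the TeriyakiTech medium male face accessory, defining bodies and joints for its two dynamic ear flaps.
# Each flap body has two capsules, alongside single-capsule bodies for the head, upper spine and both upper arms.
# Constraints join each flap to the head with one wide swing axis and tight second swing and twist limits.
# Left and right differ in capsule sizes and limits, with the right joint tighter.
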PhysicsAsset /BRCosmetics/Characters/Player/Male/Medium/Bodies/M_MED_TeriyakiTech/Meshes/Parts/M_MED_TeriyakiTech_FaceAcc_Physics_Targeting
bodies {
 dyn_earFlap_l [
  {shapeType=Capsule, length=6, radius=2, shapeName=dyn_earFlap_l_capsule, center=<3.6496 -1 0>}
  {shapeType=Capsule, length=6, radius=3, shapeName=dyn_earFlap_01_l_capsule, center=<10.877 -1.5 0>}
 ]
 dyn_earFlap_r [
  {shapeType=Capsule, length=6, radius=4.5, shapeName=dyn_earFlap_01_r_capsule, center=<12.0207 -2.8666 0>}
  {shapeType=Capsule, length=5, radius=2, shapeName=dyn_earFlap_r_capsule, center=<4.4522 -0.1282 0.3468>}
 ]
 head [{shapeType=Capsule, length=24.6488, radius=11.5, center=<0 2.3335 0>, rotation=<rot -90 180 180>}]
 spine_05 [{shapeType=Capsule, length=7.75, radius=7.255, shapeName=spine_05_capsule, center=<4.9509 -8.5287 0>, rotation=<rot 0 -90 0>}]
 upperarm_l [{shapeType=Capsule, length=21.75, radius=6, shapeName=upperarm_l_capsule, center=<12.3274 -2.894 0>, rotation=<rot 90 180 179.9998>}]
 upperarm_r [{shapeType=Capsule, length=23.75, radius=8, shapeName=upperarm_r_capsule, center=<-12.3274 2 4>, rotation=<rot 90 0 -0.0005>}]
}
constraints [
 {bone1Name=dyn_earFlap_l, bone2Name=head, swing1LimitDegrees=45, swing2LimitDegrees=10, twistLimitDegrees=5}
 {bone1Name=dyn_earFlap_r, bone2Name=head, swing1LimitDegrees=45, swing2LimitDegrees=1, twistLimitDegrees=1}
]
